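# Secondary interact prompt widget for Immunity Carriage that shows buying dropped Reboot Cards with gold.
# On context update it shows or collapses, then splits SecondContextText on "/" into gold owned ("INF" means infinite), cost, card count, rebootable players and action text.
# These values drive squad count, cost switcher, colors and dimming.
# It also rebinds to the receiving actor's MultiInteract component to update the multi-interact display, unbinding on Destruct.
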WidgetBlueprint /ImmunityCarriage/UI/WBP_ImmunityCarriage_SecondaryInteractWidget
graphs [
 BindMultiInteractUpdate {nodes [
   BindMultiInteractUpdate() {is=title K2Node_FunctionEntry}
   n1 = Get CurrentInteraction() -> (CurrentInteraction:Fort Interact Context Info Object Reference)
   > n2 = Delta Gate Wildcard!(
    execute:Exec
    Variable:"Fort Actor Component Multi Interact Object Reference (by ref)"=Get BoundMultiInteractComponent() -> (BoundMultiInteractComponent:Fort Actor Component Multi Interact Object Reference)
    New Value:Fort Actor Component Multi Interact Object Reference=GetComponentByClass!(
     self:Actor Object Reference=Get ReceivingActor!(self:Fort Interact Context Info Object Reference=@n1) -> (ReceivingActor:Actor Weak object)
     ComponentClass:Actor Component Class Reference=/Script/FortniteGame.FortActorComponent_MultiInteract
    ) -> (ReturnValue:Fort Actor Component Multi Interact Object Reference)
   ) -> (
    Success:Exec
    Updated Variable:"Fort Actor Component Multi Interact Object Reference (by ref)"
    Previous Variable:Fort Actor Component Multi Interact Object Reference
   ) {is=title K2Node_MacroInstance}
   > n3 = Is Valid(InputObject=@n2.Previous Variable)
   > n4 = UnbindMultiInteractUpdate()
   n5 = Get!(execute:Exec=[@n4.then, @n3.Is Not Valid]) -> (then:Exec, else:Exec, BoundMultiInteractComponent:Fort Actor Component Multi Interact Object Reference)
   > Bind Event to Actor Count Changed Delegate(self:Fort Actor Component Multi Interact Object Reference=@n5, Delegate=Create Event()) {is=title K2Node_AddDelegate}
   > CheckCurrentInteractionCount!(
    execute
    self:Multi Interact Display Widget Base Object Reference=Get MultiInteractDisplayWidget() -> (MultiInteractDisplayWidget:Multi Interact Display Widget Object Reference)
    CurrentCount:int=GetInteractingActorCount!(self:Fort Actor Component Multi Interact Object Reference=@n5, bIncludePlayers:bool=true, bIncludeNonPlayers:bool=true) -> (ReturnValue:int=0)
   ) -> (then)
   > n6 = Cast To MultiInteractActorComponent!(execute, Object=Get BoundMultiInteractComponent()) -> (then, CastFailed, AsMulti Interact Actor Component:Multi Interact Actor Component Object Reference) {is=title K2Node_DynamicCast}
   > SetRequiredDuration!(
    execute
    self:Multi Interact Actor Component Object Reference=@n6
    NewRequiredDuration:double=Get RequiredDuration!(self:Fort Interact Context Info Object Reference=Get CurrentInteraction()) -> (RequiredDuration:float)
   ) -> (then)
  ]}
 EventGraph {nodes [
   Event Destruct()
   > UnbindMultiInteractUpdate()
   n1 = Event OnContextInfoUpdated!() -> (OutputDelegate, then, ContextInfo:Fort Interact Context Info Object Reference) {is=title K2Node_Event}
   > Is Valid(InputObject=@n1.ContextInfo)
   > n2 = Set CurrentInteraction(CurrentInteraction:Fort Interact Context Info Object Reference=@n1.ContextInfo) -> (Output_Get:Fort Interact Context Info Object Reference)
   > n3 = Sequence()
   > SetVisibility(InVisibility=Select(
     Option 0=Collapsed
     Option 1=SelfHitTestInvisible
     Index=Get bShowSecondInteraction!(self:Fort Interact Context Info Object Reference=@n2) -> (bShowSecondInteraction:bool)
    ))
   > n4 = ParseSecondContextTest!(
    execute
    self
    InText:Text=Get SecondContextText!(self:Fort Interact Context Info Object Reference=Get CurrentInteraction()) -> (SecondContextText:Text)
   ) -> (
    then
    GoldOwned:int=0
    GoldCost:int=0
    RebootCardsAvailableForUse:int=0
    NumRebootablePlayers:int=0
    DisplayText:Text
    IsResourcesInfinite:bool
   )
   > SetCostsAndCounts!(
    execute
    self
    GoldOwned:int=@n4
    GoldCost:int=@n4.GoldCost
    RebootCardsAvailable:int=@n4.RebootCardsAvailableForUse
    RebootablePlayers:int=@n4.NumRebootablePlayers
    IsResourcesInfinite:bool=@n4.IsResourcesInfinite
   ) -> (then)
   > UpdateActionText!(execute, self, InText:Text=@n4.DisplayText) -> (then)
   BindMultiInteractUpdate(execute=@n3.then_1) {is=title K2Node_CallFunction}
   > OnMultiInteractCountChanged!(
    execute
    self
    InteractingActorCount:int=GetInteractActorCount!(
     CurrentInteraction:Fort Interact Context Info Object Reference=Get CurrentInteraction()
     MultiInteractComponent:Fort Actor Component Multi Interact Object Reference=Get BoundMultiInteractComponent()
    ) -> (InteractionCount:int=0)
    MultiInteractMultiplier:float=0
    InteractingDuration:float=0
   ) -> (then)
  ]}
 GetInteractActorCount {nodes [
   n1 = GetInteractActorCount!() -> (
    then
    CurrentInteraction:Fort Interact Context Info Object Reference
    MultiInteractComponent:Fort Actor Component Multi Interact Object Reference
   ) {is=title K2Node_FunctionEntry}
   > n2 = Branch(Condition="float > float"(A=Get RequiredDuration(self=@n1)))
   > n3 = Get!(execute) -> (then, else, MultiInteractComponent:Fort Actor Component Multi Interact Object Reference)
   > n4 = Branch(Condition=IsMultipleInteractionEnabled!(self:Fort Actor Component Multi Interact Object Reference=@n3) -> (ReturnValue:bool))
   > Return Node!(execute, InteractionCount:int=GetInteractingActorCount(self=@n3, bIncludePlayers=true, bIncludeNonPlayers=true) -> (ReturnValue=0))
   Return Node(execute=@n4.else, InteractionCount=-1)
   Return Node(execute=@n3.else, InteractionCount=-1)
   Return Node(execute=@n2.else, InteractionCount=-1)
  ]}
 OnMultiInteractCountChanged {nodes [
   n1 = OnMultiInteractCountChanged!() -> (then, InteractingActorCount:int, MultiInteractMultiplier:float=~, InteractingDuration:float=~) {is=title K2Node_FunctionEntry}
   > CheckCurrentInteractionCount(self=Get MultiInteractDisplayWidget(), CurrentCount=@n1)
  ]}
 ParseSecondContextTest {nodes [
   n1 = ParseSecondContextTest!() -> (then, InText:Text) {is=title K2Node_FunctionEntry}
   > n2 = Sequence()
   > n3 = ParseIntoArray!(execute:Exec, SourceString="To String (Text)"(InText=@n1), Delimiter=/, CullEmptyStrings) -> (then:Exec, ReturnValue)
   > n4 = Sequence!(execute) -> (then_0, then_1, then_2:Exec, then_3:Exec)
   > Branch(Condition=AND Boolean(A=Is Valid Index!(TargetArray=@n3, IndexToTest:int) -> (ReturnValue:bool), B=Is Valid Index(TargetArray=@n3, IndexToTest=1)))
   n5 = "Get (a ref)"(Array:Array of Strings=@n3, Dimension 1=0) -> (Output:"String (by ref)")
   > Set RetGoldOwned(RetGoldOwned:int=String To Integer(InString=@n5)) -> (Output_Get:int=0)
   > Set RetGoldCost(RetGoldCost:int=String To Integer(InString="Get (a ref)"(Array=@n3, Dimension 1=1))) -> (Output_Get:int=0)
   > Set RetIsResourceInfinite(RetIsResourceInfinite:bool="Equal, Case Insensitive (String)"(A=@n5, B=INF)) -> (Output_Get:bool)
   Branch(execute=@n4.then_1, Condition=Is Valid Index(TargetArray=@n3, IndexToTest=2))
   > Set RetNumRebootCardsAvailable(RetNumRebootCardsAvailable:int=String To Integer(InString="Get (a ref)"(Array=@n3, Dimension 1=2))) -> (Output_Get:int=0)
   Branch(execute=@n4.then_2, Condition=Is Valid Index(TargetArray=@n3, IndexToTest=3))
   > Set RetNumRebootablePlayers(RetNumRebootablePlayers:int=String To Integer(InString="Get (a ref)"(Array=@n3, Dimension 1=3))) -> (Output_Get:int=0)
   Branch(execute=@n4.then_3, Condition=Is Valid Index(TargetArray=@n3, IndexToTest=4))
   > Set RetDisplayText(RetDisplayText:Text="To Text (String)"(InString="Get (a ref)"(Array=@n3, Dimension 1=4))) -> (Output_Get:Text)
   Return Node!(
    execute=@n2.then_1
    GoldOwned:int=Get RetGoldOwned() -> (RetGoldOwned:int=0)
    GoldCost:int=Get RetGoldCost() -> (RetGoldCost:int=0)
    RebootCardsAvailableForUse:int=Get RetNumRebootCardsAvailable() -> (RetNumRebootCardsAvailable:int=0)
    NumRebootablePlayers:int=Get RetNumRebootablePlayers() -> (RetNumRebootablePlayers:int=0)
    DisplayText:Text=Get RetDisplayText() -> (RetDisplayText:Text)
    IsResourcesInfinite:bool=Get RetIsResourceInfinite() -> (RetIsResourceInfinite:bool)
   )
  ]}
 SetCostVisibilitySwitcher {nodes [
   n1 = SetCostVisibilitySwitcher!() -> (then, CanPurchase:bool=~) {is=title K2Node_FunctionEntry}
   > SetActiveWidget(
    self=Get VisSwitcher_Cost() -> (VisSwitcher_Cost:Common Visibility Switcher Object Reference)
    Widget=Select(
     Option 0:Widget Object Reference=Get HBox_RequiredResources() -> (HBox_RequiredResources:Horizontal Box Object Reference)
     Option 1:Widget Object Reference=Get HBox_Cost() -> (HBox_Cost:Horizontal Box Object Reference)
     Index=@n1
    ) -> (ReturnValue:Widget Object Reference)
   )
  ]}
 SetCostsAndCounts {nodes [
   SetCostsAndCounts!() -> (then, GoldOwned:int, GoldCost:int, RebootCardsAvailable:int, RebootablePlayers:int, IsResourcesInfinite:bool=~) {is=title K2Node_FunctionEntry}
   > n1 = Sequence!(execute) -> (then_0, then_1)
   > UpdateRebootableSquadCount!(
    execute
    self
    RebootablePlayers:int=Get RebootablePlayers() -> (RebootablePlayers:int=0)
    AvailableCards:int=Get RebootCardsAvailable() -> (RebootCardsAvailable:int=0)
   ) -> (then)
   n2 = Branch(execute=@n1.then_1, Condition="Equal (Integer)"(A=Get RebootCardsAvailable() -> (RebootCardsAvailable=0)))
   n3 = Get VisSwitcher_Cost()
   > SetVisibility(self=@n3, InVisibility=Collapsed)
   n4 = Get RebootablePlayers() -> (RebootablePlayers=0)
   > SetRenderOpacity(
    self=[
     Get Text_Action() -> (Text_Action:Common Text Object Reference)
     Get HBox_RebootableSquad() -> (HBox_RebootableSquad:Horizontal Box Object Reference)
     Get MultiInteractDisplayWidget()
    ]
    InOpacity=SelectFloat(A=1, B=0.5, bPickA="integer > integer"(A=@n4))
   )
   SetVisibility(execute=@n2.else, self=@n3, InVisibility=HitTestInvisible)
   > UpdateGoldCost!(
    execute
    self
    GoldOwned:int=Get GoldOwned() -> (GoldOwned:int=0)
    GoldCost:int=Get GoldCost() -> (GoldCost:int=0)
    IsResourcesInfinite:bool=Get IsResourcesInfinite() -> (IsResourcesInfinite:bool)
   ) -> (then)
   > UpdatePurchaseableRebootCards!(execute, self, InAvailableCards:int=Get RebootCardsAvailable() -> (RebootCardsAvailable=0)) -> (then)
   n5 = OR Boolean(A="integer >= integer"(A=Get GoldOwned() -> (GoldOwned=0), B=Get GoldCost() -> (GoldCost=0)), B=Get IsResourcesInfinite())
   > SetCostVisibilitySwitcher!(execute, self, CanPurchase:bool=@n5) -> (then) {is=title K2Node_CallFunction}
   n6 = "integer > integer"(A=Get RebootablePlayers() -> (RebootablePlayers=0))
   > SetRenderOpacity(
    self=[Get HBox_RebootableSquad(), Get MultiInteractDisplayWidget(), Get Text_Action()]
    InOpacity=SelectFloat(A=1, B=0.5, bPickA=OR Boolean(A=@n5, B=@n6))
   )
  ]}
 UnbindMultiInteractUpdate {nodes [
   UnbindMultiInteractUpdate() {is=title K2Node_FunctionEntry}
   > n1 = Get!(execute) -> (then, else, BoundMultiInteractComponent:Fort Actor Component Multi Interact Object Reference)
   > Unbind Event from Actor Count Changed Delegate(self:Fort Actor Component Multi Interact Object Reference=@n1, Delegate=Create Event()) {is=title K2Node_RemoveDelegate}
   > Set BoundMultiInteractComponent(BoundMultiInteractComponent:Fort Actor Component Multi Interact Object Reference) -> (Output_Get:Fort Actor Component Multi Interact Object Reference)
  ]}
 UpdateActionText {nodes [
   n1 = UpdateActionText!() -> (then, InText:Text) {is=title K2Node_FunctionEntry}
   > "SetText (Text)"(self=Get Text_Action(), InText=@n1)
  ]}
 UpdateGoldCost {nodes [
   n1 = UpdateGoldCost!() -> (then, GoldOwned:int, GoldCost:int, IsResourcesInfinite:bool=~) {is=title K2Node_FunctionEntry}
   n2 = GetInfinitySymbolText!() -> (ReturnValue:"Text (by ref)")
   n3 = "To Text (Integer)"(Value=@n1)
   > "SetText (Text)"(
    self=Get Text_GoldCost() -> (Text_GoldCost:Common Text Object Reference)
    InText=Format Text!(
     Format="LOCTEXT(\"{Owned} / {Amount}\")"
     Owned:Text=Select(Option 0:Text=@n3, Option 1:Text=@n2, Index=@n1.IsResourcesInfinite) -> (ReturnValue:Text)
     Amount:int=@n1.GoldCost
    ) -> (Result)
   )
   n4 = "integer >= integer"(A=Get GoldOwned() -> (GoldOwned=0), B=Get GoldCost() -> (GoldCost=0))
   n5 = Get IsResourcesInfinite()
   > SetColorAndOpacity(
    self=Get Text_GoldCost()
    InColorAndOpacity=Select(
     Option 0:Slate Color Structure=Get RebootCardCountZeroColor() -> (RebootCardCountZeroColor:Slate Color Structure)
     Option 1:Slate Color Structure=Get RebootCardCountNonZeroColor() -> (RebootCardCountNonZeroColor:Slate Color Structure)
     Index=OR Boolean(A=@n4, B=@n5)
    ) -> (ReturnValue:Slate Color Structure)
   )
   > "SetText (Text)"(
    self=Get Text_RequiredGoldCost() -> (Text_RequiredGoldCost:Common Text Object Reference)
    InText="To Text (Integer)"(Value=Get GoldCost() -> (GoldCost=0))
   )
  ]}
 UpdatePurchaseableRebootCards {nodes [
   n1 = UpdatePurchaseableRebootCards!() -> (then, InAvailableCards:int) {is=title K2Node_FunctionEntry}
   n2 = "To Text (Integer)"(Value=@n1)
   > "SetText (Text)"(self=Get Text_RebootCardToPurchase() -> (Text_RebootCardToPurchase:Common Text Object Reference), InText=@n2)
   > "SetText (Text)"(self=Get Text_RequiredCards() -> (Text_RequiredCards:Common Text Object Reference), InText=@n2)
  ]}
 UpdateRebootableSquadCount {nodes [
   n1 = UpdateRebootableSquadCount!() -> (then, RebootablePlayers:int, AvailableCards:int) {is=title K2Node_FunctionEntry}
   n2 = int + int(A=@n1, B=@n1.AvailableCards)
   > "SetText (Text)"(
    self=Get Text_RebootablePlayers() -> (Text_RebootablePlayers:Common Text Object Reference)
    InText=Format Text!(
     Format="LOCTEXT(\"{Players}/{Total Needed}\")"
     Players:Text="To Text (Integer)"(Value=@n1)
     Total Needed:Text="To Text (Integer)"(Value=@n2)
    ) -> (Result)
   )
  ]}
 UpdateRequiredResources {nodes [
   n1 = UpdateRequiredResources!() -> (then, GoldCost:int, RequiredCards:int) {is=title K2Node_FunctionEntry}
   > "SetText (Text)"(self=Get Text_RequiredGoldCost(), InText="To Text (Integer)"(Value=@n1))
   > "SetText (Text)"(self=Get Text_RequiredCards(), InText="To Text (Integer)"(Value=@n1.RequiredCards))
  ]}
]
variables [
 BoundMultiInteractComponent {type=FortActorComponent_MultiInteract, default=null}
 CurrentInteraction {type=FortInteractContextInfo, default=null}
 RebootCardCountNonZeroColor {type=SlateColor, default {colorUseRule=UseColor_Specified, specifiedColor=<rgba 1 1 1 1>}}
 RebootCardCountZeroColor {type=SlateColor, default {colorUseRule=UseColor_Specified, specifiedColor=<rgba 1 0.0588 0.0667 1>}}
]
widgets {
 info {inheritedWidgetCount=0, parentClass=/Script/FortniteGame.FortInteractContextInfoWidget, rootWidgetClass=&Border, widgetCount=25}
 widgets [
  Border_Root {
   widgetClassPath=&Border
   widget=Border_Root
   properties {
    background {
     bIsDynamicallyLoaded=false
     drawAs=Image
     imageType=NoImage
     mirroring=NoMirror
     resourceName=None
     resourceObject=null
     tiling=NoTile
     imageSize=<32 32>
     margin=<ltrb 0 0 0 0>
     outlineSettings {
      bUseBrushTransparency=false
      roundingType=HalfHeightRadius
      width=0
      color {colorUseRule=UseColor_Specified, specifiedColor=<rgba 0 0 0 0>}
      cornerRadii=<0 0 0 0>
     }
     tintColor {colorUseRule=UseColor_Specified, specifiedColor=<rgba 1 1 1 0>}
     uVRegion {bIsValid=false, max=<0 0>, min=<0 0>}
    }
    padding=<ltrb 8 0 0 0>
   }
  }
  VerticalBox_50 {
   widgetClassPath=&VerticalBox
   parent=Border_Root
   widget=VerticalBox_50
   slot {class=BorderSlot, properties {padding=<ltrb 8 0 0 0>}}
  }
  HBox_Interaction {
   widgetClassPath=&HorizontalBox
   parent=VerticalBox_50
   widget=HBox_Interaction
   slot {class=VerticalBoxSlot, properties {verticalAlignment=VAlign_Top}}
  }
  Text_Action {
   widgetClassPath=&CommonTextBlock
   bIsVariable=true
   parent=HBox_Interaction
   widget=Text_Action
   properties {
    lineHeightPercentage=0.9
    text=Buy Dropped Reboot Cards
    textTransformPolicy=ToUpper
    font {
     bForceMonospaced=false
     bMaterialIsStencil=false
     fontMaterial=null
     fontObject=&Burbank
     letterSpacing=0
     monospacedWidth=1
     size=16
     skewAmount=0
     typefaceFontName=BurbankRegular-Black
     outlineSettings {
      bApplyOutlineToDropShadows=false
      bMiteredCorners=false
      bSeparateFillAlpha=true
      outlineMaterial=&M-FontShadow
      outlineSize=2
      outlineColor=<rgba 0 0 0 0.5>
     }
    }
    shadowOffset=<0 0>
   }
   slot {class=HorizontalBoxSlot, properties {horizontalAlignment=HAlign_Left, verticalAlignment=VAlign_Center}}
  }
  Spacer_91 {widgetClassPath=&Spacer, parent=HBox_Interaction, widget=Spacer_91, properties {size=<10 1>}, slot {class=HorizontalBoxSlot}}
  HBox_RebootableSquad {
   widgetClassPath=&HorizontalBox
   bIsVariable=true
   parent=HBox_Interaction
   widget=HBox_RebootableSquad
   slot {class=HorizontalBoxSlot, properties {horizontalAlignment=HAlign_Center, verticalAlignment=VAlign_Center}}
  }
  Image_PlayerIcon {
   widgetClassPath=&FortLazyImage
   bIsVariable=true
   parent=HBox_RebootableSquad
   widget=Image_PlayerIcon
   properties {brush {
     bIsDynamicallyLoaded=false
     drawAs=Image
     imageType=FullColor
     mirroring=NoMirror
     resourceName=None
     resourceObject=/Game/Athena/HUD/Spectator/Art/T-UI-Icon-PlayersRemaining-64x64
     tiling=NoTile
     imageSize=<24 24>
     margin=<ltrb 0 0 0 0>
     outlineSettings {
      bUseBrushTransparency=false
      roundingType=HalfHeightRadius
      width=0
      color {colorUseRule=UseColor_Specified, specifiedColor=<rgba 0 0 0 0>}
      cornerRadii=<0 0 0 0>
     }
     tintColor {colorUseRule=UseColor_Specified, specifiedColor=<rgba 1 1 1 1>}
     uVRegion {bIsValid=false, max=<0 0>, min=<0 0>}
    }}
   slot {class=HorizontalBoxSlot, properties {verticalAlignment=VAlign_Center}}
  }
  Text_RebootablePlayers {
   widgetClassPath=&CommonTextBlock
   bIsVariable=true
   parent=HBox_RebootableSquad
   widget=Text_RebootablePlayers
   properties {
    mobileFontSizeMultiplier=1.12
    text=0/0
    font {
     bForceMonospaced=false
     bMaterialIsStencil=false
     fontMaterial=null
     fontObject=&BurbankSmall
     letterSpacing=0
     monospacedWidth=1
     size=16
     skewAmount=0
     typefaceFontName=Black
     outlineSettings {
      bApplyOutlineToDropShadows=false
      bMiteredCorners=false
      bSeparateFillAlpha=true
      outlineMaterial=&M-FontShadow
      outlineSize=3
      outlineColor=<rgba 0 0 0 0.5>
     }
    }
    shadowOffset=<0 0>
   }
   slot {class=HorizontalBoxSlot}
  }
  HorizontalBox_2 {
   widgetClassPath=&HorizontalBox
   parent=VerticalBox_50
   widget=HorizontalBox_2
   slot {class=VerticalBoxSlot, properties {verticalAlignment=VAlign_Bottom}}
  }
  MultiInteractDisplayWidget {
   widgetClassPath=/Game/UI/InGame/HUD/MultiInteract/MultiInteractDisplayWidget
   bIsVariable=true
   parent=HorizontalBox_2
   widget=MultiInteractDisplayWidget
   slot {class=HorizontalBoxSlot, properties {verticalAlignment=VAlign_Center, padding=<ltrb 0 0 5 0>}}
  }
  VisSwitcher_Cost {
   widgetClassPath=&CommonVisibilitySwitcher
   bIsVariable=true
   parent=HorizontalBox_2
   widget=VisSwitcher_Cost
   slot {class=HorizontalBoxSlot, properties {size {sizeRule=Fill, value=1}}}
  }
  HBox_Cost {
   widgetClassPath=&HorizontalBox
   bIsVariable=true
   parent=VisSwitcher_Cost
   widget=HBox_Cost
   slot {class=CommonVisibilitySwitcherSlot}
  }
  HBox_Gold {
   widgetClassPath=&HorizontalBox
   bIsVariable=true
   parent=HBox_Cost
   widget=HBox_Gold
   slot {class=HorizontalBoxSlot, properties {horizontalAlignment=HAlign_Left}}
  }
  Image_Gold {
   widgetClassPath=&FortLazyImage
   bIsVariable=true
   parent=HBox_Gold
   widget=Image_Gold
   properties {brush {
     bIsDynamicallyLoaded=false
     drawAs=Image
     imageType=FullColor
     mirroring=NoMirror
     resourceName=None
     resourceObject=/Game/Athena/HUD/Inventory/Art/M_UI_HUD_Inventory_ResourceIcon_Currency
     tiling=NoTile
     imageSize=<32 32>
     margin=<ltrb 0 0 0 0>
     outlineSettings {
      bUseBrushTransparency=false
      roundingType=HalfHeightRadius
      width=0
      color {colorUseRule=UseColor_Specified, specifiedColor=<rgba 0 0 0 0>}
      cornerRadii=<0 0 0 0>
     }
     tintColor {colorUseRule=UseColor_Specified, specifiedColor=<rgba 1 1 1 1>}
     uVRegion {bIsValid=false, max=<0 0>, min=<0 0>}
    }}
   slot {class=HorizontalBoxSlot, properties {verticalAlignment=VAlign_Center}}
  }
  Text_GoldCost {
   widgetClassPath=&CommonTextBlock
   bIsVariable=true
   parent=HBox_Gold
   widget=Text_GoldCost
   properties {
    lineHeightPercentage=0.9
    mobileFontSizeMultiplier=1.12
    text="{Have} / {Need}"
    font {
     bForceMonospaced=false
     bMaterialIsStencil=false
     fontMaterial=null
     fontObject=&BurbankSmall
     letterSpacing=0
     monospacedWidth=1
     size=13.5
     skewAmount=0
     typefaceFontName=Black
     outlineSettings {
      bApplyOutlineToDropShadows=false
      bMiteredCorners=false
      bSeparateFillAlpha=true
      outlineMaterial=&M-FontShadow
      outlineSize=2
      outlineColor=<rgba 0 0 0 0.5>
     }
    }
    shadowOffset=<0 0>
   }
   slot {class=HorizontalBoxSlot, properties {verticalAlignment=VAlign_Center}}
  }
  HBox_RebootCards {
   widgetClassPath=&HorizontalBox
   bIsVariable=true
   parent=HBox_Cost
   widget=HBox_RebootCards
   slot {class=HorizontalBoxSlot, properties {horizontalAlignment=HAlign_Right, size {sizeRule=Fill, value=1}}}
  }
  Image_RebootCard {
   widgetClassPath=&FortLazyImage
   parent=HBox_RebootCards
   widget=Image_RebootCard
   properties {brush {
     bIsDynamicallyLoaded=false
     drawAs=Image
     imageType=FullColor
     mirroring=NoMirror
     resourceName=None
     resourceObject=/Game/Athena/HUD/Materials/MI_UI_RebootCardIcon
     tiling=NoTile
     imageSize=<32 32>
     margin=<ltrb 0 0 0 0>
     outlineSettings {
      bUseBrushTransparency=false
      roundingType=HalfHeightRadius
      width=0
      color {colorUseRule=UseColor_Specified, specifiedColor=<rgba 0 0 0 0>}
      cornerRadii=<0 0 0 0>
     }
     tintColor {colorUseRule=UseColor_Specified, specifiedColor=<rgba 1 1 1 1>}
     uVRegion {bIsValid=false, max=<0 0>, min=<0 0>}
    }}
   slot {class=HorizontalBoxSlot, properties {verticalAlignment=VAlign_Center}}
  }
  Text_RebootCardToPurchase {
   widgetClassPath=&CommonTextBlock
   bIsVariable=true
   parent=HBox_RebootCards
   widget=Text_RebootCardToPurchase
   properties {
    lineHeightPercentage=0.9
    mobileFontSizeMultiplier=1.12
    text=x0
    font {
     bForceMonospaced=false
     bMaterialIsStencil=false
     fontMaterial=null
     fontObject=&BurbankSmall
     letterSpacing=0
     monospacedWidth=1
     size=13.5
     skewAmount=0
     typefaceFontName=Black
     outlineSettings {
      bApplyOutlineToDropShadows=false
      bMiteredCorners=false
      bSeparateFillAlpha=true
      outlineMaterial=&M-FontShadow
      outlineSize=2
      outlineColor=<rgba 0 0 0 0.5>
     }
    }
    renderTransformPivot=<0 0.5>
    shadowOffset=<0 0>
   }
   slot {class=HorizontalBoxSlot, properties {verticalAlignment=VAlign_Center}}
  }
  HBox_RequiredResources {
   widgetClassPath=&HorizontalBox
   bIsVariable=true
   parent=VisSwitcher_Cost
   widget=HBox_RequiredResources
   slot {class=CommonVisibilitySwitcherSlot, properties {horizontalAlignment=HAlign_Right}}
  }
  HBox_RequiredResourcesForCard {
   widgetClassPath=&HorizontalBox
   bIsVariable=true
   parent=HBox_RequiredResources
   widget=HBox_RequiredResourcesForCard
   slot {class=HorizontalBoxSlot, properties {size {sizeRule=Fill, value=1}}}
  }
  Image_RequiredGold {
   widgetClassPath=&FortLazyImage
   bIsVariable=true
   parent=HBox_RequiredResourcesForCard
   widget=Image_RequiredGold
   properties {brush {
     bIsDynamicallyLoaded=false
     drawAs=Image
     imageType=FullColor
     mirroring=NoMirror
     resourceName=None
     resourceObject=&M_UI_HUD_Inventory_ResourceIcon_Currency
     tiling=NoTile
     imageSize=<32 32>
     margin=<ltrb 0 0 0 0>
     outlineSettings {
      bUseBrushTransparency=false
      roundingType=HalfHeightRadius
      width=0
      color {colorUseRule=UseColor_Specified, specifiedColor=<rgba 0 0 0 0>}
      cornerRadii=<0 0 0 0>
     }
     tintColor {colorUseRule=UseColor_Specified, specifiedColor=<rgba 1 1 1 1>}
     uVRegion {bIsValid=false, max=<0 0>, min=<0 0>}
    }}
   slot {class=HorizontalBoxSlot}
  }
  Text_RequiredGoldCost {
   widgetClassPath=&CommonTextBlock
   bIsVariable=true
   parent=HBox_RequiredResourcesForCard
   widget=Text_RequiredGoldCost
   properties {
    lineHeightPercentage=0.9
    mobileFontSizeMultiplier=1.12
    text=300
    colorAndOpacity {colorUseRule=UseColor_Specified, specifiedColor=<rgba 1 0.0588 0.0667 1>}
    font {
     bForceMonospaced=false
     bMaterialIsStencil=false
     fontMaterial=null
     fontObject=&Burbank
     letterSpacing=0
     monospacedWidth=1
     size=16
     skewAmount=0
     typefaceFontName=BurbankRegular-Black
     outlineSettings {
      bApplyOutlineToDropShadows=false
      bMiteredCorners=false
      bSeparateFillAlpha=true
      outlineMaterial=&M-FontShadow
      outlineSize=2
      outlineColor=<rgba 0 0 0 0.5>
     }
    }
    renderTransform {angle=0, scale=<0.8 0.8>, shear=<0 0>, translation=<0 0>}
    renderTransformPivot=<0 0.5>
    shadowOffset=<0 0>
   }
   slot {class=HorizontalBoxSlot}
  }
  Text_Or {
   widgetClassPath=&CommonTextBlock
   parent=HBox_RequiredResourcesForCard
   widget=Text_Or
   properties {
    lineHeightPercentage=0.9
    mobileFontSizeMultiplier=1.12
    text=Or
    textTransformPolicy=ToLower
    font {
     bForceMonospaced=false
     bMaterialIsStencil=false
     fontMaterial=null
     fontObject=&Burbank
     letterSpacing=0
     monospacedWidth=1
     size=16
     skewAmount=0
     typefaceFontName=BurbankRegular-Black
     outlineSettings {
      bApplyOutlineToDropShadows=false
      bMiteredCorners=false
      bSeparateFillAlpha=true
      outlineMaterial=&M-FontShadow
      outlineSize=2
      outlineColor=<rgba 0 0 0 0.5>
     }
    }
    renderTransform {angle=0, scale=<0.8 0.8>, shear=<0 0>, translation=<0 0>}
    renderTransformPivot=<0 0.5>
    shadowOffset=<0 0>
   }
   slot {class=HorizontalBoxSlot}
  }
  Image_RequiredCard {
   widgetClassPath=&FortLazyImage
   parent=HBox_RequiredResources
   widget=Image_RequiredCard
   properties {brush {
     bIsDynamicallyLoaded=false
     drawAs=Image
     imageType=FullColor
     mirroring=NoMirror
     resourceName=None
     resourceObject=&MI_UI_RebootCardIcon
     tiling=NoTile
     imageSize=<32 32>
     margin=<ltrb 0 0 0 0>
     outlineSettings {
      bUseBrushTransparency=false
      roundingType=HalfHeightRadius
      width=0
      color {colorUseRule=UseColor_Specified, specifiedColor=<rgba 0 0 0 0>}
      cornerRadii=<0 0 0 0>
     }
     tintColor {colorUseRule=UseColor_Specified, specifiedColor=<rgba 1 1 1 1>}
     uVRegion {bIsValid=false, max=<0 0>, min=<0 0>}
    }}
   slot {class=HorizontalBoxSlot}
  }
  Text_RequiredCards {
   widgetClassPath=&CommonTextBlock
   bIsVariable=true
   parent=HBox_RequiredResources
   widget=Text_RequiredCards
   properties {
    lineHeightPercentage=0.9
    mobileFontSizeMultiplier=1.12
    text=3
    colorAndOpacity {colorUseRule=UseColor_Specified, specifiedColor=<rgba 1 0.0588 0.0667 1>}
    font {
     bForceMonospaced=false
     bMaterialIsStencil=false
     fontMaterial=null
     fontObject=&Burbank
     letterSpacing=0
     monospacedWidth=1
     size=16
     skewAmount=0
     typefaceFontName=BurbankRegular-Black
     outlineSettings {
      bApplyOutlineToDropShadows=false
      bMiteredCorners=false
      bSeparateFillAlpha=true
      outlineMaterial=&M-FontShadow
      outlineSize=2
      outlineColor=<rgba 0 0 0 0.5>
     }
    }
    renderTransform {angle=0, scale=<0.8 0.8>, shear=<0 0>, translation=<0 0>}
    renderTransformPivot=<0 0.5>
    shadowOffset=<0 0>
   }
   slot {class=HorizontalBoxSlot}
  }
 ]
}
